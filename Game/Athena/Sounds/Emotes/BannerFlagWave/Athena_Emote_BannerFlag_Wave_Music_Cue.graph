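# Sound cue that plays the music for the Banner Flag Wave emote in Athena.
# A mixer starts a one-shot first wave immediately alongside a looping second wave delayed by about one second, forming an intro-into-loop sequence.
# It uses the emote music sound class and 1P emote music concurrency rule, and has a send to the music audio analysis submix.
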SoundCue /Game/Athena/Sounds/Emotes/BannerFlagWave/Athena_Emote_BannerFlag_Wave_Music_Cue
root=Mixer_0
nodes [
 {id=Delay_0, children [WavePlayer_2], properties {delayMax=1.002, delayMin=1.002}}
 {id=Mixer_0, children [WavePlayer_3,Delay_0], properties {inputVolume [1,1]}}
 {id=WavePlayer_2, properties {bLooping=true, soundWaveAssetPtr=&/Emote_Banner_Flag_Wave_02}}
 {id=WavePlayer_3, properties {soundWaveAssetPtr=&/Emote_Banner_Flag_Wave_01}}
]
properties {
 maxDistance=2097152
 soundClassObject=&Fort_Music_Emote
 volumeMultiplier=1
 concurrencyOverrides {maxCount=3, resolutionRule=StopOldest}
 concurrencySet [/Game/Sounds/Concurrency_Rules/Athena_EmoteMusic_1P_Conc]
 soundSubmixSends [{minSendLevel=0, soundSubmix=&Fortnite_AudioAnalysis_Music}]
}
